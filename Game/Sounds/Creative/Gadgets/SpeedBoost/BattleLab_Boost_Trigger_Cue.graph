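# Sound cue for the trigger of the Creative Battle Lab speed boost gadget, reusing trap and hoverboard sounds.
# A Mixer layers a random pick of two Speed Trap fire waves, at reduced level, with a Hoverboard boost wave.
# A Modulator varies pitch slightly, then an Enveloper holds volume before fading out.
# The cue plays pitched up and quieter, with trap attenuation and sound class, and per-owner concurrency that stops the oldest.
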SoundCue /Game/Sounds/Creative/Gadgets/SpeedBoost/BattleLab_Boost_Trigger_Cue
root=Enveloper_0
nodes [
 {
  id=Enveloper_0
  children [Modulator_6]
  properties {
   pitchCurve {editorCurveData {keys [{time=0}]}}
   volumeCurve {editorCurveData {keys [<at 0 1>, {arriveTangent=-0.3636, interpMode=RCIM_Cubic, leaveTangent=-0.3636, tangentMode=RCTM_Break, time=2, value=1}, <at 3 0>]}}
  }
 }
 {id=Mixer_0, children [Random_13,WavePlayer_0], properties {inputVolume [0.5,1]}}
 {id=Modulator_6, children [Mixer_0], properties {pitchMax=1.1, pitchMin=0.9}}
 {id=Random_13, children [WavePlayer_53, WavePlayer_54], properties {weights [1,1]}}
 {id=WavePlayer_0, properties {soundWaveAssetPtr=/Game/Athena/Sounds/Vehicles/Hoverboard/Hoverboard_Boost_01}}
 {id=WavePlayer_53, properties {soundWaveAssetPtr=/Game/Sounds/Fort_Traps/Speed_Trap/Speed_Trap_Fire02}}
 {id=WavePlayer_54, properties {soundWaveAssetPtr=/Game/Sounds/Fort_Traps/Speed_Trap/Speed_Trap_Fire03}}
]
properties {
 attenuationSettings=&Fort_Traps_Att
 bOverrideConcurrency=true
 maxDistance=6000
 pitchMultiplier=1.8
 soundClassObject=/Game/Sounds/FortSoundClassesAndModes/Fort_Traps_Placement
 volumeMultiplier=0.35
 attenuationOverrides {bAttenuateWithLPF=true, falloffDistance=3100, lPFRadiusMin=2500, attenuationShapeExtents {x=900}}
 concurrencyOverrides {bLimitToOwner=true, maxCount=2, resolutionRule=StopOldest}
}
