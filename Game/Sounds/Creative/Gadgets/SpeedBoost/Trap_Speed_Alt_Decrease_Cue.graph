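# Sound cue for the alternate speed-decrease sound of the Creative speed boost trap gadget, reusing a gravity sphere stop wave.
# The wave player feeds a default-setting Modulator, then an Enveloper whose volume curve fades it to silence over under two seconds.
# It plays at reduced volume in the Creative gadgets sound class, with gadget-activate attenuation and a dedicated speed-boost-decrease concurrency set.
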SoundCue /Game/Sounds/Creative/Gadgets/SpeedBoost/Trap_Speed_Alt_Decrease_Cue
root=Enveloper_0
nodes [
 {
  id=Enveloper_0
  children [Modulator_6]
  properties {pitchCurve {editorCurveData {keys [{time=0}]}}, volumeCurve {editorCurveData {keys [<at 0 1>,<at 1.7 0>]}}}
 }
 {id=Modulator_6, children [WavePlayer_30]}
 {id=WavePlayer_30, properties {soundWaveAssetPtr=/Game/Sounds/Fort_Gadgets/GravitySphere/fort_gravity_sphere_stop_02}}
]
properties {
 attenuationSettings=&Fort_GadgetActivate_Att
 maxDistance=4000
 soundClassObject=&Fort_Gadgets_Creative
 volumeMultiplier=0.6
 concurrencyOverrides {bLimitToOwner=true, maxCount=1}
 concurrencySet [/Game/Sounds/Creative/Concurrency/Creative_SpeedBoost_Decrease_Conc]
}
