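# Sound cue for the Creative speed boost gadget's stop sound on the alternate decrease (slow) variant.
# It plays a single slow-stop wave through a Modulator node at reduced volume.
# It uses gadget activation attenuation and the Creative gadgets sound class.
# A speed boost decrease concurrency set applies, with an override limiting it to one instance per owner.
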SoundCue /Game/Sounds/Creative/Gadgets/SpeedBoost/Trap_Speed_Alt_Decrease_Stop_Cue
root=Modulator_6
nodes [
 {id=Modulator_6, children [WavePlayer_1]}
 {id=WavePlayer_1, properties {soundWaveAssetPtr=&/CM_SpeedBoost_Slow_Stop01}}
]
properties {
 attenuationSettings=&Fort_GadgetActivate_Att
 maxDistance=4000
 priority=0.5
 soundClassObject=&Fort_Gadgets_Creative
 volumeMultiplier=0.6
 concurrencyOverrides {bLimitToOwner=true, maxCount=1}
 concurrencySet [/Game/Sounds/Creative/Concurrency/Creative_SpeedBoost_Decrease_Conc]
}
