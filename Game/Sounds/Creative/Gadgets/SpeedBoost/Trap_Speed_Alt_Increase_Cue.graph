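# Sound cue for the Creative speed boost gadget's alternate speed increase sound, reusing a hoverboard boost wave.
# A wave player feeds an enveloper that fades volume to silence over 1.75 seconds, with slight random pitch and volume variation.
# The cue plays in the Creative gadgets sound class with distant golf cart attenuation.
# It references a speed boost concurrency asset and also carries override values for one instance per owner.
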SoundCue /Game/Sounds/Creative/Gadgets/SpeedBoost/Trap_Speed_Alt_Increase_Cue
root=Enveloper_0
nodes [
 {
  id=Enveloper_0
  children [WavePlayer_1]
  properties {
   pitchMax=1.05
   pitchMin=0.95
   volumeMax=1.05
   volumeMin=0.95
   pitchCurve {editorCurveData {keys [{time=0}]}}
   volumeCurve {editorCurveData {keys [<at 0 1>,<at 1.75 0>]}}
  }
 }
 {id=WavePlayer_1, properties {soundWaveAssetPtr=/Game/Athena/Sounds/Vehicles/Hoverboard/Hoverboard_Boost_01}}
]
properties {
 attenuationSettings=/Game/Sounds/Attenuation/Fort_GolfCart_Movement_Distant_Att
 maxDistance=10100
 soundClassObject=&Fort_Gadgets_Creative
 volumeMultiplier=0.7
 concurrencyOverrides {bLimitToOwner=true, maxCount=1}
 concurrencySet [/Game/Sounds/Creative/Concurrency/Creative_SpeedBoost_Increase_Conc]
}
