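# Sound cue for the stop of a Creative speed boost trap's alternate increase effect, reusing the jetpack stop wave.
# An Enveloper wraps the single wave player, fading volume to silence over 1.5 seconds with slightly raised, randomized pitch and slight volume variation.
# It plays at reduced volume with jetpack attenuation, in the Creative gadgets sound class.
# It references a speed boost concurrency set and carries a one-per-owner concurrency override.
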SoundCue /Game/Sounds/Creative/Gadgets/SpeedBoost/Trap_Speed_Alt_Increase_Stop_Cue
root=Enveloper_0
nodes [
 {
  id=Enveloper_0
  children [WavePlayer_0]
  properties {
   pitchMax=1.2
   pitchMin=1.1
   volumeMax=1.05
   volumeMin=0.95
   pitchCurve {editorCurveData {keys [{time=0}]}}
   volumeCurve {editorCurveData {keys [<at 0 1>,<at 1.5 0>]}}
  }
 }
 {id=WavePlayer_0, properties {soundWaveAssetPtr=/Game/Athena/Sounds/Items/Jetpack/Jetpack_Stop_01}}
]
properties {
 attenuationSettings=/Game/Sounds/Attenuation/Fort_Jetpack_Att
 maxDistance=13067
 priority=0.5
 soundClassObject=&Fort_Gadgets_Creative
 volumeMultiplier=0.4
 concurrencyOverrides {bLimitToOwner=true, maxCount=1}
 concurrencySet [/Game/Sounds/Creative/Concurrency/Creative_SpeedBoost_Increase_Conc]
}
